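# Master hair material for the Zombie Elastic male medium player body, adding a recolorable tint to the shared FN_Hair function.
# It multiplies the boosted HSV value of FN_Hair's base color by HairColor, then blends toward flat HairColor using the mask's blue channel.
# An optional ClothFuzz switch, using the mask's green channel, feeds MF_AnimatedHair.
# The result passes through tech-art overrides, de-rez transition and character opacity functions.
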Material /BRCosmetics/Characters/Player/Male/Medium/Bodies/M_MED_Zombie_Elastic/Materials/M_FN_Hair_ZombieElastic_MASTER
graph {
 properties {
  bUseMaterialAttributes=true
  bUsedWithMeshDeformer=true
  bUsedWithMorphTargets=true
  bUsedWithSkeletalMesh=true
  pixelDepthOffsetMode=PDOM_Legacy
  shadingModel=MSM_FromMaterialExpression
  usedShadingModels=Default Lit | Cloth
 }
 nodes [
  n1 = ScalarParameter() {defaultValue=1, group=Elastic, parameterName=HairClothFuzzAlpha, sliderMax=1}
  n2 = TextureSampleParameter2D() {
   group=Hair
   parameterName=M
   samplerType=SAMPLERTYPE_Masks
   texture=/BRCosmetics/Characters/Player/Male/Medium/Bodies/M_MED_Zombie_Elastic/Skins/CAU/Textures/T_M_MED_Zombie_Elastic_CAU_Hair_M
  }
  n3 = MaterialFunctionCall() {materialFunction=/Game/Packages/Fortress_SharedMaterials/Base_Material_Functions/FN_Hair}
  n4 = GetMaterialAttributes(None=@n3.Result) {attributeGetTypes [#1]}
  n5 = Multiply(A=ComponentMask(None=MaterialFunctionCall!(In=@n4.Base Color) {materialFunction=/Engine/Functions/Engine_MaterialFunctions01/Chromakeying/RGBtoHSV}.Result) {b=true}) {constB=3}
  n6 = VectorParameter() {group=Elastic, parameterName=HairColor, defaultValue=<rgba 1 0 0 1>}
  n7 = SetMaterialAttributes!(MaterialAttributes=@n3.Result, Base Color=LinearInterpolate(A=Multiply(A=@n5, B=@n6.RGB), B=@n6.RGB, Alpha=Multiply(A=@n2.B) {constB=0.5})) {attributeSetTypes [#1]}
  n8 = MaterialFunctionCall!(
   GMap
   Material=StaticSwitchParameter(
    True=MaterialFunctionCall!(Alpha=@n1, Gmap=@n2.G, Material=@n7) {materialFunction=/Game/Weapons/WeaponSkins/MaterialFunction/ClothFuzz}.Result
    False=@n7
   ) {group=Skin FX, parameterName=Use Cloth fuzz}
  ) {materialFunction=/Game/Weapons/WeaponSkins/MaterialFunction/MF_AnimatedHair}
  MaterialFunctionCall!(Material=MaterialFunctionCall!(
    Fresnel=Fresnel()
    Material=MaterialFunctionCall(In=@n8.Result) {materialFunction=&MF_TechArtOverrides}.Result
   ) {materialFunction=/Game/Packages/Fortress_SharedMaterials/Base_Material_Functions/CharacterDeRez_Transition}.Result) {materialFunction=/Game/Packages/Fortress_SharedMaterials/Base_Material_Functions/BaseCharacterShaderFunctions/BaseFN_CharacterOpacity}
 ]
}
